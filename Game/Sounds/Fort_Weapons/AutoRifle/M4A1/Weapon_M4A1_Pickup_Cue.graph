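# Sound cue for picking up the M4A1 auto rifle.
# A single wave player reuses the Bullpup pickup recording, slightly pitched down and played at reduced volume.
# It routes through the loot collect sound class with shared weapon pickup attenuation.
# A concurrency override allows one instance at a time and blocks new plays while one is active.
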SoundCue /Game/Sounds/Fort_Weapons/AutoRifle/M4A1/Weapon_M4A1_Pickup_Cue
root=WavePlayer_2
nodes [{id=WavePlayer_2, properties {soundWaveAssetPtr=/Game/Sounds/Fort_Weapons/AutoRifle/Bullpup/Bullpup_Pickup_01}}]
properties {
 attenuationSettings=&Fort_Weapon_Pickup_Att
 bOverrideConcurrency=true
 maxDistance=3500
 pitchMultiplier=0.95
 soundClassObject=&Fort_Loot_Collect
 volumeMultiplier=0.45
 concurrencyOverrides {maxCount=1, resolutionRule=PreventNew}
}
